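# Masked, two-sided lit Niagara ribbon material for the Venus skydiving contrail, drawing an inky smoke trail.
# Panned smoke texture drives a two-color gradient with optional time-of-day and particle tints, while its normal map supports two-sided correction.
# Opacity combines smoke, particle alpha and U-edge fade, with switches for near-camera cutout and temporal dithering.
# A sine-distorted local-space UV branch exists but does not feed the texture samples.
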Material /BRCosmetics/Athena/Effects/Particles/Skydiving/NS_Contrail_Venus/Materials/M_MeshTrail_Original
graph {
 properties {
  bEnableResponsiveAA=true
  bUsedWithNiagaraRibbons=true
  blendMode=BLEND_Masked
  opacityMaskClipValue=0.15
  pixelDepthOffsetMode=PDOM_Legacy
  translucencyLightingMode=TLM_VolumetricDirectional
  twoSided=true
  usedShadingModels=Default Lit
 }
 nodes [
  n1 = Constant3Vector() {constant=<rgba 1 0.6459 0.4526 0>}
  n2 = MaterialFunctionCall(Color_Morning=@n1, Color_Evening=@n1) {materialFunction=&Color_Time_of_Day}
  n3 = VectorParameter() {parameterName="Color_Morning (V3)"}
  n4 = VectorParameter() {parameterName="Color_Day (V3)"}
  n5 = VectorParameter() {parameterName="Color_Evening (V3)"}
  n6 = VectorParameter() {parameterName="Color_Night (V3)"}
  n7 = TextureCoordinate() {coordinateIndex=1}
  n8 = MaterialFunctionCall!(Texture Coordinates=@n7, U Frequency, V Frequency, Offset, OffsetMultiplier) {materialFunction=/Game/Effects/Niagara/MasterMaterials/Functions/MF_Niagara_Panner}
  n9 = TextureSample(UVs=@n8.Result) {texture=/BRCosmetics/Athena/Effects/Textures/T_InkySmokeTrail}
  n10 = ScalarParameter() {defaultValue=1, group=ColorTint, parameterName=Power}
  n11 = LinearInterpolate(
   A=VectorParameter() {group=ColorTint, parameterName=ColorLow, defaultValue=<rgba 0.0118 0.0833 0.0155 1>}.RGB
   B=VectorParameter() {group=ColorTint, parameterName=ColorHigh, defaultValue=<rgba 0.4844 0.4118 0 1>}.RGB
   Alpha=Subtract(A=Power(Base=@n9.R, Exp=@n10) {constExponent=0.5}, B=ScalarParameter() {defaultValue=0.25, group=ColorTint, parameterName=HighColorRange}) {constB=0.25}
  )
  n12 = StaticSwitchParameter(
   True=Multiply(
    A=MaterialFunctionCall(Color_Morning=@n3.RGB, Color_Day=@n4.RGB, Color_Evening=@n5.RGB, Color_Night=@n6.RGB) {materialFunction=&Color_Time_of_Day}.Result
    B=@n11
   )
   False=@n11
  ) {desc=New, group=ColorTint, parameterName=UseTimeOfDay}
  n13 = ParticleColor()
  n14 = StaticSwitchParameter(True=Multiply(A=@n12, B=@n13.RGB), False=@n12) {desc=New, group=ColorTint, parameterName=UseParticleColorTint}
  MaterialFunctionCall(Emissive=StaticSwitchParameter(True=Multiply(A=@n2.Result, B=@n14), False=Constant()) {desc=New, parameterName=EmissiveOnly}) {materialFunction=&MF_FX_EmissiveAdjustments}
  n15 = TextureSample(UVs=@n8.Result) {samplerType=SAMPLERTYPE_Normal, texture=/BRCosmetics/Weapons/FORT_Melee/Pickaxe_Venus/T_InkySmokeTrail_Normal}
  n16 = MaterialFunctionCall(Float3=@n15.RGB) {materialFunction=&BreakOutFloat3Components}
  n17 = MaterialFunctionCall(X=@n16.G, Y=@n16.G, Z=@n16.G) {materialFunction=&MakeFloat3}
  n18 = TwoSidedSign()
  n19 = ScalarParameter() {parameterName=NormalFlatness}
  Reroute(None=StaticSwitchParameter(
    True=LinearInterpolate(A=@n17.Result, B=@n15.RGB, Alpha=@n18)
    False=MaterialFunctionCall(Normal=@n15.RGB, Flatness=@n19) {materialFunction=&FlattenNormal}.Result
   ) {defaultValue=true, parameterName=UseTwoSideSign})
  ScalarParameter() {defaultValue=0.05, parameterName=Specular}
  ScalarParameter() {defaultValue=0.1, parameterName=Metallic}
  ScalarParameter() {defaultValue=0.5, parameterName=Roughness}
  n20 = ComponentMask(None=TextureCoordinate() {uTiling=2}) {r=true}
  n21 = OneMinus(None=Abs(None=OneMinus(None=@n20)))
  n22 = Multiply(A=Multiply(A=@n9.R, B=@n13.A), B=Saturate(None=Multiply(A=@n21) {constB=10}))
  n23 = SphereMask(A=WorldPosition().XYZ, B=CameraPositionWS(), Radius=Constant() {r=150}, Hardness=Constant() {r=0.1})
  n24 = Multiply(A=@n23, B=@n23)
  n25 = Subtract(A=@n22, B=Saturate(None=Multiply(A=@n24, B=@n23)))
  n26 = Multiply(
   A=StaticSwitchParameter(True=Saturate(None=@n25), False=@n22) {desc=New, parameterName=AsSkydive}
   B=ScalarParameter() {defaultValue=1, parameterName=AlphaIntensity}
  ) {constB=3}
  StaticSwitchParameter(
   True=Multiply(
    A=@n26
    B=MaterialFunctionCall(Alpha Threshold=@n26) {materialFunction=&DitherTemporalAA_Fortnite}.Result
   )
   False=@n26
  ) {desc=New, parameterName=UseTemporalAA}
  n27 = MaterialFunctionCall!(MaxiumURange=Constant() {r=1000}, RangeRampHardness, MirrorTexCoordU) {materialFunction=/BRCosmetics/Weapons/FORT_Melee/Pickaxe_Venus/MF_LocalSpaceTrailUV}
  n28 = ComponentMask(None=@n27."LocalSpaceCoordinate(UV)") {r=true}
  n29 = ScalarParameter() {defaultValue=2, parameterName=TillingU}
  n30 = StaticSwitchParameter(True=AppendVector(A=Multiply(A=@n28, B=@n29), B=ComponentMask(None=@n27."LocalSpaceCoordinate(UV)") {g=true}), False=@n7) {parameterName=UseLocalUV}
  n31 = ComponentMask(None=@n30) {r=true}
  n32 = DynamicParameter() {paramNames [Time,Random,Panner,Param4]}
  n33 = ComponentMask(None=@n30) {g=true}
  n34 = ScalarParameter() {defaultValue=2, parameterName=WaveF}
  n35 = Multiply(A=Add(A=@n31, B=@n32.Time), B=@n34) {constB=10}
  n36 = Multiply(A=Divide(A=Sine(None=@n35), B=@n34) {constB=10}) {constB=0.25}
  n37 = AppendVector(
   A=Add(A=@n31, B=Add(A=@n32.Random, B=@n32.Random))
   B=LinearInterpolate(A=@n33, B=Add(A=@n33, B=@n36), Alpha=ScalarParameter() {defaultValue=1, parameterName=DistortIntensity}) {desc=New}
  )
  n38 = ScalarParameter() {parameterName=SpeedU}
  n39 = ScalarParameter() {parameterName=SpeedV}
  StaticSwitchParameter(True=@n37, False=Panner(Coordinate=@n37, Speed=AppendVector(A=@n38, B=@n39)) {desc=New, speedX=-1}) {parameterName=IngoneRedundantPanner}
 ]
}
