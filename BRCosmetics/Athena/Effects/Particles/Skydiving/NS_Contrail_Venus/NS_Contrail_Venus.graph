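# Niagara skydiving contrail cosmetic "Venus": a ribbon trail with flowers, leaves and glows attached to the player's skeletal mesh.
# Emitters pick bones by spawn group from User.SkeletalMesh, with skydive spawn-rate modules driving all but the burst-spawned, pulsing glow sprites.
# Flower meshes tumble and scale by curve, while two leaf sprite emitters spin, one also orbiting.
# Ribbons lock their first particle to the bone and switch lifetime and velocity for front end.
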NiagaraSystem /BRCosmetics/Athena/Effects/Particles/Skydiving/NS_Contrail_Venus/NS_Contrail_Venus
dependencies {
 usedDataInterfaces [&NiagaraDataInterfaceCurve, &NiagaraDataInterfaceVectorCurve]
 usedDynamicInputs []
 usedModules [&SystemState]
}
emitters {
 BadFlower_Inherited {
  emitterSpawnScript [&FN_PreserveSkydiveUserVariables() {module=PreserveSkydiveUserVariables}]
  emitterUpdateScript [
   &EmitterScalability(
    Module.bApplySpawnCountScale=false
    Module.bApplySpawnCountScaleByDistanceFraction=false
    Module.bUseMaxDistance=false
    Module.bUseMinDistance=false
   ) {enabled=false}
   &EmitterLifeCycle() {enabled=false}
   &EmitterState()
   /Game/Effects/Niagara/Modules/Spawn/FN_SkydiveSpawnRate(Global Spawnrate=3) {module=SkydiveSpawnRate}
  ]
  particleSpawnScript [
   /Game/Effects/Niagara/Modules/Groups/FN_SetSpawnGroup_v1/FN_SetSpawnGroup(Override SpawnGroup=true, SpawnGroup:NiagaraInt32=&SpawnGroup())
   /Niagara/Modules/Spawn/Initialization/InitializeParticle(
    Lifetime=&UniformRangedFloat(Minimum=0.5, Maximum=0.7)
    Mesh Scale=<0.0003 0.0003 0.0003>
    Module.Write Color=true
    Module.Write Lifetime=true
    Module.Write Mass=false
    Module.Write Material Random=true
    Module.Write Position=false
    Module.Write Scale=true
    Module.Write SpriteRotation=false
    Module.Write SpriteSize=false
    Position=&SimulationPosition()
   )
   /Niagara/Modules/Update/Orientation/InitialMeshRotationRate(
    Initial Rotation Rate Scale:NiagaraFloat=&UniformRangedFloat(Minimum=0, Maximum=1)
    Pitch:NiagaraFloat=&UniformRangedFloat(Minimum=0, Maximum=360)
    Roll:NiagaraFloat=&UniformRangedFloat(Minimum=0, Maximum=360)
    Yaw:NiagaraFloat=&UniformRangedFloat(Minimum=0, Maximum=360)
   )
   &AddVelocityInCone(
    Cone Angle=80
    Cone Axis=<0 0 1>
    Module.Use Velocity Falloff On Cone Axis=false
    Velocity Strength=&UniformRangedFloat(Minimum=50, Maximum=120)
   )
   &SampleSkeletalMeshSkeleton(
    Mesh:NiagaraDataInterfaceSkeletalMesh=$User.SkeletalMesh
    Specific Bone Index:NiagaraInt32=$Particles.SpawnGroup
    Specific Socket Index:NiagaraInt32=$Particles.SpawnGroup
   )
   /Niagara/Modules/Spawn/Location/SkeletalMeshSkeletonLocation(Module.Inherit Velocity=false, Module.Use Sampled Bone as Ribbon ID=true)
  ]
  particleUpdateScript [
   &ParticleState()
   &Color(Color=<rgba 4.3662 3.0843 1.8045 1>)
   &ScaleMeshSize(Scale Factor=&VectorFromFloat(Value=&FloatFromCurve(
      FloatCurve=DataInterface() {propertyValues {Curve {keys [
          <at 0 0>
          {arriveTangent=0.7948, interpMode=RCIM_Cubic, leaveTangent=0.7948, tangentMode=RCTM_User, time=0.1089, value=0.416}
          {arriveTangent=0.3617, interpMode=RCIM_Cubic, leaveTangent=0.3617, tangentMode=RCTM_User, time=0.5681, value=0.5849}
          {arriveTangent=6.0683, interpMode=RCIM_Cubic, leaveTangent=-6.1941, tangentMode=RCTM_Break, time=0.6672, value=1.0049}
          {arriveTangent=-3.3458, interpMode=RCIM_Cubic, leaveTangent=-3.3458, time=0.7315, value=0.477}
          {arriveTangent=-0.6359, interpMode=RCIM_Cubic, leaveTangent=-0.6359, tangentMode=RCTM_User, time=0.8633, value=0.3489}
          <at 1.0011 0.002>
         ]}}}
      Scale Curve=1000
     )))
   /Niagara/Modules/Update/Orientation/MeshRotationRate(Pitch:NiagaraFloat=$Particles.InitialPitchRate, Roll:NiagaraFloat=$Particles.InitialRollRate, Yaw:NiagaraFloat=$Particles.InitialYawRate)
   &SolveForcesAndVelocity(Clamp Velocity=false, Limit Acceleration=false)
   /Niagara/Modules/Update/Material/DynamicMatParams(
    Module.Param0WriteEnabled=true
    Module.Param1WriteEnabled=true
    Module.Param2WriteEnabled=true
    Module.Param3WriteEnabled=true
    Param0Value:NiagaraFloat=&FloatFromCurve(FloatCurve=DataInterface() {propertyValues {Curve {keys [<at 0 -0.1>,<at 0.42 -0.3>,<at 0.4982 0.5067>,<at 0.6 -0.32>,<at 0.66 -0.32>,<at 1.0011 0.002>]}}})
    Param1Value=1
    Param2Value=1
    Param3Value=1
   )
  ]
  rendererClasses [&NiagaraMeshRendererProperties]
  renderers [{rendererClass=&NiagaraMeshRendererProperties, rendererIndex=0}]
 }
 Glow_Inherited {
  emitterSpawnScript [&FN_PreserveSkydiveUserVariables() {module=PreserveSkydiveUserVariables}]
  emitterUpdateScript [
   &EmitterState()
   &EmitterLifeCycle(CompleteOnInactive=true) {enabled=false}
   &SpawnBurst_Instantaneous(Spawn Count=4, Use Spawn Probability=false)
  ]
  particleSpawnScript [
   &FN_SetSpawnGroup(Emitter Bone Count=4)
   &InitializeParticle(
    Color=<rgba 0.2 0.1939 0.0567 0.05>
    Lifetime=100
    Module.Write Color=true
    Module.Write Lifetime=true
    Module.Write Mass=false
    Module.Write Material Random=true
    Module.Write Position=false
    Module.Write Scale=false
    Module.Write SpriteRotation=false
    Module.Write SpriteSize=true
    Position=&SimulationPosition()
    Sprite Size=<25 25>
   )
  ]
  particleUpdateScript [
   &SampleSkeletalMeshSkeleton(Mesh:NiagaraDataInterfaceSkeletalMesh=$User.SkeletalMesh, Specific Bone Index:NiagaraInt32=$Particles.SpawnGroup)
   &SkeletalMeshSkeletonLocation(Module.Inherit Velocity=false, Module.Use Sampled Bone as Ribbon ID=true)
   &ParticleState()
   &ScaleSpriteSize(Scale Factor=&Vector2DFromFloat(Value=/BRCosmetics/Athena/Effects/Particles/Skydiving/NS_Contrail_Raven/Module/Pulse(
      A:NiagaraFloat=1.0547
      Amplitude:NiagaraFloat=0.2049
      Crest:NiagaraFloat=1.4786
      Phase:NiagaraFloat=&Multiply_Float(A=$Emitter.Age, B=0.5)
      PhaseOffset:NiagaraFloat=0
     )))
   &CameraOffset(Camera Offset Amount=50)
  ]
  rendererClasses [&NiagaraSpriteRendererProperties]
  renderers [{rendererClass=&NiagaraSpriteRendererProperties}]
 }
 Leaf_GPU_Inherited {
  emitterSpawnScript [&FN_PreserveSkydiveUserVariables() {module=PreserveSkydiveUserVariables}]
  emitterUpdateScript [
   &EmitterScalability(
    Module.bApplySpawnCountScale=false
    Module.bApplySpawnCountScaleByDistanceFraction=false
    Module.bUseMaxDistance=true
    Module.bUseMinDistance=false
    SpawnCountScale=0.5
    SpawnCountScaleByDistanceFraction=DataInterface() {propertyValues {Curve {keys [<at 0 1>,<at 1 0>]}}}
   ) {enabled=false}
   &EmitterLifeCycle() {enabled=false}
   &EmitterState()
   &FN_SkydiveSpawnRate(Global Spawnrate=4, Jitter 0=1) {module=SkydiveSpawnRate}
   &SpawnPerUnit(Module.Use Max Movement Threshold=true, Module.Use Movement Tolerance=false, Spawn Per Unit=250, Use Spawn Probability=false)
  ]
  particleSpawnScript [
   &FN_SetSpawnGroup(Override SpawnGroup=true, SpawnGroup:NiagaraInt32=&SpawnGroup())
   &InitializeParticle(
    Module.Write Color=true
    Module.Write Lifetime=true
    Module.Write Mass=false
    Module.Write Material Random=true
    Module.Write Position=false
    Module.Write Scale=false
    Module.Write SpriteRotation=false
    Module.Write SpriteSize=true
    Position=&SimulationPosition()
    Sprite Size=&Vector2DFromFloat(Value=&UniformRangedFloat(Minimum=6, Maximum=12))
   )
   /SetVariables(Particles.SpriteInitRotationRate:NiagaraFloat=&UniformRangedFloat(Minimum=360, Maximum=720)) {isSetParameters=true}
   /SetVariables(Particles.SubImageIndex=HlslExpression() {hlslExpression="floor(Particles.RandomFloat)"}) {enabled=false, isSetParameters=true}
   &AddVelocity(Velocity=&UniformRangedVector(Minimum=<0 0 75>, Maximum=<1 1 150>))
   &CylinderLocation(
    Cylinder Height=1
    Cylinder Radius=10
    Lathe Profile=DataInterface() {propertyValues {Curve {keys []}}}
    MaskSpawnGroup=false
    Module.Enable Lathe Profile=false
    Module.Override Randomness=false
    Module.Surface Only=false
   )
   &SampleSkeletalMeshSkeleton(
    Mesh:NiagaraDataInterfaceSkeletalMesh=$User.SkeletalMesh
    Specific Bone Index:NiagaraInt32=$Particles.SpawnGroup
    Specific Socket Index:NiagaraInt32=$Particles.SpawnGroup
   )
   &SkeletalMeshSkeletonLocation(Module.Inherit Velocity=false, Module.Use Sampled Bone as Ribbon ID=true)
  ]
  particleUpdateScript [
   &ParticleState()
   &ScaleSpriteSize(Scale Factor=&Vector2DFromFloat(Value=&FloatFromCurve(FloatCurve=DataInterface() {propertyValues {Curve {keys [<at 0 0 RCIM_Cubic>, {arriveTangent=0.002, interpMode=RCIM_Cubic, leaveTangent=0.002, time=0.1216, value=0.8846}, <at 1.0011 0.002>]}}})))
   &Color(
    Color=<rgba 1.156 2 1.3065 1>
    Scale Alpha=&FloatFromCurve(FloatCurve=DataInterface() {propertyValues {Curve {keys [<at -0.0046 1.0113>,<at 1.0011 0.002>]}}})
   )
   &SolveForcesAndVelocity(Clamp Velocity=false, Limit Acceleration=false)
   &SpriteRotationRate(Rotation Rate=$Particles.SpriteInitRotationRate)
  ]
  rendererClasses [&NiagaraSpriteRendererProperties]
  renderers [{rendererClass=&NiagaraSpriteRendererProperties}]
 }
 Leaf_Orbit_Inherited {
  emitterSpawnScript [&FN_PreserveSkydiveUserVariables() {module=PreserveSkydiveUserVariables}]
  emitterUpdateScript [
   &EmitterScalability(
    MaxDistance=1000
    Module.bApplySpawnCountScale=false
    Module.bApplySpawnCountScaleByDistanceFraction=true
    Module.bUseMaxDistance=true
    Module.bUseMinDistance=false
    SpawnCountScaleByDistanceFraction=DataInterface() {propertyValues {Curve {keys [<at -0.0317 1>,<at 4.0412 8.0588>]}}}
   ) {enabled=false}
   &EmitterLifeCycle() {enabled=false}
   &EmitterState()
   &FN_SkydiveSpawnRate(Global Spawnrate=4) {module=SkydiveSpawnRate}
  ]
  particleSpawnScript [
   &FN_SetSpawnGroup(Override SpawnGroup=true, SpawnGroup:NiagaraInt32=&SpawnGroup())
   &InitializeParticle(
    Lifetime=&UniformRangedFloat(Minimum=0.35, Maximum=0.5)
    Module.Write Color=true
    Module.Write Lifetime=true
    Module.Write Mass=false
    Module.Write Material Random=true
    Module.Write Position=false
    Module.Write Scale=false
    Module.Write SpriteRotation=false
    Module.Write SpriteSize=true
    Position=&SimulationPosition()
    Sprite Rotation=&UniformRangedFloat(Minimum=0, Maximum=1)
    Sprite Size=&Vector2DFromFloat(Value=&UniformRangedFloat(Minimum=3.75, Maximum=11.25))
   )
   /SetVariables(Particles.InitRotationRate:NiagaraFloat=&NormalizedAngleToDegrees(NormalizedAngle=&UniformRangedFloat(Minimum=2, Maximum=4))) {isSetParameters=true}
   /SetVariables(
    Particles.Orbit.RotationOffset:NiagaraFloat=&UniformRangedFloat(Minimum=-1, Maximum=1)
    Particles.Orbit.RotationRate:NiagaraFloat=&UniformRangedFloat(Minimum=-1, Maximum=1)
   ) {isSetParameters=true}
   &AddVelocityInCone(
    Cone Axis=<0 0 1>
    Module.Use Velocity Falloff On Cone Axis=false
    Velocity Strength=/Niagara/DynamicInputs/Random/FixedSeedRandomFloat(Seed:NiagaraFloat=0, Range:Vector2f=<50 100>)
   )
   &SampleSkeletalMeshSkeleton(
    Mesh:NiagaraDataInterfaceSkeletalMesh=$User.SkeletalMesh
    Specific Bone Index:NiagaraInt32=$Particles.SpawnGroup
    Specific Socket Index:NiagaraInt32=$Particles.SpawnGroup
   )
   &SkeletalMeshSkeletonLocation(Module.Inherit Velocity=false, Module.Use Sampled Bone as Ribbon ID=true)
  ]
  particleUpdateScript [
   &ParticleState()
   &Color(
    Color=<rgba 1.2 1.2 1.2 1>
    Scale Alpha=&FloatFromCurve(FloatCurve=DataInterface() {propertyValues {Curve {keys [<at -0.0046 1.0113>,<at 1.0011 0.002>]}}})
   )
   &ScaleSpriteSize(Scale Factor=&Vector2DFromFloat(Value=&FloatFromCurve(FloatCurve=DataInterface() {propertyValues {Curve {keys [<at 0 0>,<at 0.0524 1>,<at 0.85 1>,<at 1.0011 0.002>]}}})))
   &SolveForcesAndVelocity(Clamp Velocity=false, Limit Acceleration=false)
   /BRCosmetics/Effects/Niagara/Modules/Orbit/FN_Orbit(Radius:NiagaraFloat=&FloatFromCurve(
     FloatCurve=DataInterface() {propertyValues {Curve {keys [{arriveTangent=3.0355, interpMode=RCIM_Cubic, leaveTangent=3.0355, tangentMode=RCTM_User, time=-0.0046, value=0}, <at 1.0011 1 RCIM_Cubic>]}}}
     Scale Curve=20
    ))
   &SpriteRotationRate(Rotation Rate=$Particles.InitRotationRate)
  ]
  rendererClasses [&NiagaraSpriteRendererProperties]
  renderers [{rendererClass=&NiagaraSpriteRendererProperties}]
 }
 Ribbon {
  emitterSpawnScript [&FN_PreserveSkydiveUserVariables() {module=PreserveSkydiveUserVariables}]
  emitterUpdateScript [
   &EmitterScalability(
    Module.bApplySpawnCountScale=false
    Module.bApplySpawnCountScaleByDistanceFraction=false
    Module.bUseMaxDistance=false
    Module.bUseMinDistance=false
   )
   &EmitterLifeCycle() {enabled=false}
   &EmitterState()
   &FN_SkydiveSpawnRate(Global Spawnrate=90) {module=SkydiveSpawnRate}
  ]
  particleSpawnScript [
   &FN_SetSpawnGroup(Override SpawnGroup=true, SpawnGroup:NiagaraInt32=&SpawnGroup())
   /SetVariables(Particles.ExecutionIndex:NiagaraFloat=&ReturnNormalizedExecIndex()) {isSetParameters=true}
   &InitializeRibbon(
    Lifetime=&FN_FrontEndFloatModifier(Gameplay=0.2, Front End=1)
    Module.Write Color=true
    Module.Write Lifetime=true
    Module.Write Mass=true
    Module.Write Material Random=true
    Module.Write Position=false
    Module.Write RibbonTwist=false
    Module.Write RibbonWidth=true
    Position:Vector3f=&SimulationPosition()
    Ribbon Width=20
   )
   &AddVelocity(Velocity=&FN_FrontEndVector3Modifier(Gameplay=<0 0 0>, Front End=<0 0 400>))
   &SampleSkeletalMeshSkeleton(
    Mesh:NiagaraDataInterfaceSkeletalMesh=$User.SkeletalMesh
    Specific Bone Index:NiagaraInt32=$Particles.SpawnGroup
    Specific Socket Index:NiagaraInt32=$Particles.SpawnGroup
   )
   &SkeletalMeshSkeletonLocation(Module.Inherit Velocity=false, Module.Use Sampled Bone as Ribbon ID=true)
  ]
  particleUpdateScript [
   &ParticleState()
   /BRCosmetics/Effects/Niagara/Modules/Ribbons/LockFirstParticleToBone(BoneId:NiagaraInt32=&SpawnGroup(), SkeletalMesh:NiagaraDataInterfaceSkeletalMesh=$User.SkeletalMesh)
   /Game/Effects/Niagara/Modules/Materials/MaterialPanner(Frequency=1.7214, Jitter Multiplier=3, Offset=1.0294, Speed=2.4)
   &SolveForcesAndVelocity(Clamp Velocity=false, Limit Acceleration=false)
   /SetVariables(Particles.RibbonWidth=&Multiply_Float(
     A=$Particles.Initial.RibbonWidth
     B=&FloatFromCurve(FloatCurve=DataInterface() {propertyValues {Curve {keys [
          <at -0.006 0.9769 RCIM_Cubic>
          {arriveTangent=1.2519, interpMode=RCIM_Cubic, leaveTangent=1.2519, tangentMode=RCTM_User, time=0.0792, value=1.034}
         ]}}})
    )) {isSetParameters=true}
   &Color(
    Color=<rgba 1 1 0.9855 1>
    Scale Alpha=&FloatFromCurve(FloatCurve=DataInterface() {propertyValues {Curve {keys [
         {arriveTangent=0.0244, interpMode=RCIM_Cubic, leaveTangent=0.0244, tangentMode=RCTM_User, time=0.8257, value=0.9787}
         {arriveTangent=-1.503, interpMode=RCIM_Cubic, tangentMode=RCTM_Break, time=0.9944, value=0.0093}
        ]}}})
    Scale Color=&VectorFromCurve(VectorCurve=DataInterface() {propertyValues {
       XCurve {keys [{arriveTangent=3.1459, interpMode=RCIM_Cubic, leaveTangent=3.1459, tangentMode=RCTM_User, time=-0.0015, value=0.0241}, <at 0.7449 2.2116>]}
       YCurve {keys [{arriveTangent=4.9991, interpMode=RCIM_Cubic, leaveTangent=4.9991, tangentMode=RCTM_User, time=-0.0027, value=0.051}, <at 0.7473 2.082>]}
       ZCurve {keys [{arriveTangent=3.7179, interpMode=RCIM_Cubic, leaveTangent=3.7179, tangentMode=RCTM_User, time=-0.0016, value=0.0844}, <at 0.7491 2.0161>]}
      }})
   ) {enabled=false}
   &Color(
    Color=<rgba 1.25 1.25 1.25 1.5>
    Scale Alpha=&FloatFromCurve(FloatCurve=DataInterface() {propertyValues {Curve {keys [
         <at -0.0046 1.0113>
         {arriveTangent=-1.1563, interpMode=RCIM_Cubic, leaveTangent=-1.1563, time=0.25, value=1}
         <at 0.8758 -0.0067 RCIM_Cubic>
        ]}}})
    Scale Color=&VectorFromFloat(Value=&FloatFromCurve(FloatCurve=DataInterface() {propertyValues {Curve {keys [<at 0 0.35>,<at 0.3 0.27>,<at 1 1>]}}}))
   ) {module=Color001}
   &DynamicMatParams(
    Module.Param0WriteEnabled=false
    Module.Param1WriteEnabled=true
    Module.Param2WriteEnabled=true
    Module.Param3WriteEnabled=true
    Param0Value:NiagaraFloat=$Emitter.Age
    Param1Value:NiagaraFloat=&Multiply_Float(A=1, B=1)
    Param2Value:NiagaraFloat=&Multiply_Float(A=1, B=1)
    Param3Value=1
   ) {enabled=false}
   /SetVariables(Particles.RenderPosition:Vector3f=&Add_Vector(A=Linked() {linkedVariable {name=Particles.Position, type=NiagaraPosition}}, B=<0 0 -7>)) {isSetParameters=true}
  ]
  rendererClasses [&NiagaraRibbonRendererProperties]
  renderers [{rendererClass=&NiagaraRibbonRendererProperties}]
 }
}
summary {
 systemName=NS_Contrail_Venus
 emitters [
  {emitterName=Ribbon, rendererClasses [&NiagaraRibbonRendererProperties]}
  {emitterName=BadFlower_Inherited, rendererClasses [&NiagaraMeshRendererProperties]}
  {emitterName=Glow_Inherited, rendererClasses [&NiagaraSpriteRendererProperties]}
  {emitterName=Leaf_GPU_Inherited, rendererClasses [&NiagaraSpriteRendererProperties]}
  {emitterName=Leaf_Orbit_Inherited, rendererClasses [&NiagaraSpriteRendererProperties]}
 ]
 userVariables [
  User.bIsFrontEnd {type=NiagaraBool}
  User.bIsFrontEndPreview {type=NiagaraBool}
  User.PlayerSpeed {type=NiagaraFloat, defaultValue=0, description=""}
  User.SkeletalMesh {
   type=NiagaraDataInterfaceSkeletalMesh
   defaultValue {
    struct=NiagaraExt_VariableValue_DataInterface
    value {dataInterface=&self:NiagaraDataInterfaceSkeletalMesh_3}
   }
  }
 ]
}
